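# Post-copy anim blueprint for the female DangerIndex face accessory part, adding secondary physics motion to its hair and bangs.
# It copies the attached parent's pose, simulates the back hair bone, then constrains the b, c and d hair masters between head and that bone.
# Additive rotations offset three bangs before spring-driven Anim Dynamics simulates five bangs relative to the head and each hair strand bone.
# The event graph has no logic.
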AnimBlueprint /BRCosmetics/Characters/Player/Female/Medium/Bodies/F_MED_DangerIndex/Meshes/Parts/F_MED_DangerIndex_FaceAcc_AnimBP
target_skeleton=&/F_MED_DangerIndex_FaceAcc_Skeleton
graphs [
 AnimGraph {nodes [
   n1 = Anim Dynamics - Bone: dyn_hair_back_01(ComponentPose=Local To Component(LocalPose=Copy Pose From Mesh() {bUseAttachedParent=true}), Alpha=1, ExternalForce="0,0,0") {
    angularDampingOverride=0.9999
    angularSpringConstant=200
    bAngularSpring=true
    bOverrideAngularDamping=true
    bUseGravityOverride=true
    boundBone {boneName=dyn_hair_back_01}
    componentAppliedLinearAccClamp=<5000 5000 5000>
    componentLinearAccScale=<0.5 0.5 0.5>
    componentLinearVelScale=<0.5 0.5 0.5>
    gravityOverride {z=-980}
    physicsBodyDefinitions [{
      boundBone {boneName=dyn_hair_back_01}
      boxExtents=<7 7 7>
      constraintSetup {angularLimitsMax=<5 15 0>, angularLimitsMin=<-5 -15 -45>, angularTarget=<1 0 0>}
      localJointOffset=<0.2 0 0>
     }]
    simSpaceSettings {maxAngularAcceleration=32, maxAngularVelocity=8, simSpaceAngularAlpha=1}
   }
   n2 = "Constraint - dyn_hair_b_master (PP)"!(ComponentPose=@n1, Alpha=1, ConstraintWeights_0=0.5, ConstraintWeights_1:float=0.5) -> (Pose) {
    boneToModify {boneName=dyn_hair_b_master}
    constraintSetup [{targetBone {boneName=head}}, {targetBone {boneName=dyn_hair_back_01}}]
    constraintWeights [1,1]
   }
   n3 = "Constraint - dyn_hair_c_master (PP)"!(ComponentPose=@n2, Alpha=1, ConstraintWeights_0=0.5, ConstraintWeights_1:float=0.5) -> (Pose) {
    boneToModify {boneName=dyn_hair_c_master}
    constraintSetup [{targetBone {boneName=head}}, {targetBone {boneName=dyn_hair_back_01}}]
    constraintWeights [1,1]
   }
   n4 = "Constraint - dyn_hair_d_master (PP)"!(ComponentPose=@n3, Alpha=1, ConstraintWeights_0=0.5, ConstraintWeights_1:float=0.5) -> (Pose) {
    boneToModify {boneName=dyn_hair_d_master}
    constraintSetup [{targetBone {boneName=head}}, {targetBone {boneName=dyn_hair_back_01}}]
    constraintWeights [1,1]
   }
   n5 = "Transform (Modify) Bone - Bone: dyn_bang_01"(ComponentPose=@n4, Alpha=1, Translation="0,0,0", Rotation="-0.700,14.358,-2.444", Scale="1,1,1") {rotationMode=BMM_Additive, rotationSpace=BCS_ParentBoneSpace, boneToModify {boneName=dyn_bang_01}, rotation=<rot -0.7 14.358 -2.444>}
   n6 = "Transform (Modify) Bone - Bone: dyn_bang_02"(ComponentPose=@n5, Alpha=1, Translation="0,0,0", Rotation="1.681,11.570,-6.025", Scale="1,1,1") {rotationMode=BMM_Additive, rotationSpace=BCS_ParentBoneSpace, boneToModify {boneName=dyn_bang_02}, rotation=<rot 1.681 11.57 -6.025>}
   n7 = "Transform (Modify) Bone - Bone: dyn_bang_03"(ComponentPose=@n6, Alpha=1, Translation="0,0,0", Rotation="-6.222,-11.282,-10.052", Scale="1,1,1") {rotationMode=BMM_Additive, rotationSpace=BCS_ParentBoneSpace, boneToModify {boneName=dyn_bang_03}, rotation=<rot -6.222 -11.282 -10.052>}
   n8 = Anim Dynamics - Bone: dyn_bang_01(ComponentPose=@n7, Alpha=1, ExternalForce="0,0,0") {
    angularDampingOverride=1
    angularSpringConstant=2000
    bAngularSpring=true
    bOverrideAngularDamping=true
    bUseGravityOverride=true
    simulationSpace=BoneRelative
    boundBone {boneName=dyn_bang_01}
    componentAppliedLinearAccClamp=<5000 5000 5000>
    componentLinearAccScale=<0.5 0.5 0.5>
    componentLinearVelScale=<0.5 0.5 0.5>
    gravityOverride {z=-980}
    physicsBodyDefinitions [{
      boundBone {boneName=dyn_bang_01}
      boxExtents=<15 15 15>
      constraintSetup {angularLimitsMax=<0 8 7>, angularLimitsMin=<-3 -1 0>, angularTarget=<1 0 0>}
      localJointOffset=<0.09 0 0>
     }]
    relativeSpaceBone {boneName=head}
    simSpaceSettings {maxAngularAcceleration=32, maxAngularVelocity=8, simSpaceAngularAlpha=1}
   }
   n9 = Anim Dynamics - Bone: dyn_bang_02(ComponentPose=@n8, Alpha=1, ExternalForce="0,0,0") {
    angularDampingOverride=1
    angularSpringConstant=2000
    bAngularSpring=true
    bOverrideAngularDamping=true
    bUseGravityOverride=true
    simulationSpace=BoneRelative
    boundBone {boneName=dyn_bang_02}
    componentAppliedLinearAccClamp=<5000 5000 5000>
    componentLinearAccScale=<0.5 0.5 0.5>
    componentLinearVelScale=<0.5 0.5 0.5>
    gravityOverride {z=-980}
    physicsBodyDefinitions [{
      boundBone {boneName=dyn_bang_02}
      boxExtents=<15 15 15>
      constraintSetup {angularLimitsMax=<0 8 7>, angularLimitsMin=<-3 -8 0>, angularTarget=<1 0 0>}
      localJointOffset=<0.09 0 0>
     }]
    relativeSpaceBone {boneName=head}
    simSpaceSettings {maxAngularAcceleration=32, maxAngularVelocity=8, simSpaceAngularAlpha=1}
   }
   n10 = Anim Dynamics - Bone: dyn_bang_03(ComponentPose=@n9, Alpha=1, ExternalForce="0,0,0") {
    angularDampingOverride=1
    angularSpringConstant=2000
    bAngularSpring=true
    bOverrideAngularDamping=true
    bUseGravityOverride=true
    simulationSpace=BoneRelative
    boundBone {boneName=dyn_bang_03}
    componentAppliedLinearAccClamp=<5000 5000 5000>
    componentLinearAccScale=<0.5 0.5 0.5>
    componentLinearVelScale=<0.5 0.5 0.5>
    gravityOverride {z=-980}
    physicsBodyDefinitions [{
      boundBone {boneName=dyn_bang_03}
      boxExtents=<15 15 15>
      constraintSetup {angularLimitsMax=<3 8 7>, angularLimitsMin=<0 -8 0>, angularTarget=<1 0 0>}
      localJointOffset=<0.09 0 0>
     }]
    relativeSpaceBone {boneName=head}
    simSpaceSettings {maxAngularAcceleration=32, maxAngularVelocity=8, simSpaceAngularAlpha=1}
   }
   n11 = Anim Dynamics - Bone: dyn_bang_04(ComponentPose=@n10, Alpha=1, ExternalForce="0,0,0") {
    angularDampingOverride=1
    angularSpringConstant=2000
    bAngularSpring=true
    bOverrideAngularDamping=true
    bUseGravityOverride=true
    simulationSpace=BoneRelative
    boundBone {boneName=dyn_bang_04}
    componentAppliedLinearAccClamp=<5000 5000 5000>
    componentLinearAccScale=<0.5 0.5 0.5>
    componentLinearVelScale=<0.5 0.5 0.5>
    gravityOverride {z=-980}
    physicsBodyDefinitions [{
      boundBone {boneName=dyn_bang_04}
      boxExtents=<15 15 15>
      constraintSetup {angularLimitsMax=<3 8 7>, angularLimitsMin=<0 -8 0>, angularTarget=<1 0 0>}
      localJointOffset=<0.09 0 0>
     }]
    relativeSpaceBone {boneName=head}
    simSpaceSettings {maxAngularAcceleration=32, maxAngularVelocity=8, simSpaceAngularAlpha=1}
   }
   n12 = Anim Dynamics - Bone: dyn_bang_05(ComponentPose=@n11, Alpha=1, ExternalForce="0,0,0") {
    angularDampingOverride=1
    angularSpringConstant=2000
    bAngularSpring=true
    bOverrideAngularDamping=true
    bUseGravityOverride=true
    simulationSpace=BoneRelative
    boundBone {boneName=dyn_bang_05}
    componentAppliedLinearAccClamp=<5000 5000 5000>
    componentLinearAccScale=<0.5 0.5 0.5>
    componentLinearVelScale=<0.5 0.5 0.5>
    gravityOverride {z=-980}
    physicsBodyDefinitions [{
      boundBone {boneName=dyn_bang_05}
      boxExtents=<15 15 15>
      constraintSetup {angularLimitsMax=<3 8 7>, angularLimitsMin=<0 -8 0>, angularTarget=<1 0 0>}
      localJointOffset=<0.09 0 0>
     }]
    relativeSpaceBone {boneName=head}
    simSpaceSettings {maxAngularAcceleration=32, maxAngularVelocity=8, simSpaceAngularAlpha=1}
   }
   n13 = Anim Dynamics - Bone: dyn_hair_a_01(ComponentPose=@n12, Alpha=1, ExternalForce="0,0,0") {
    angularDampingOverride=0.9999
    angularSpringConstant=1000
    bAngularSpring=true
    bOverrideAngularDamping=true
    bUseGravityOverride=true
    boundBone {boneName=dyn_hair_a_01}
    componentAppliedLinearAccClamp=<5000 5000 5000>
    componentLinearAccScale=<0.5 0.5 0.5>
    componentLinearVelScale=<0.5 0.5 0.5>
    gravityOverride {z=-980}
    physicsBodyDefinitions [{
      boundBone {boneName=dyn_hair_a_01}
      boxExtents=<7 7 7>
      constraintSetup {angularLimitsMax=<5 30 30>, angularLimitsMin=<-5 -30 -30>, angularTarget=<1 0 0>}
      localJointOffset=<0.2 0 0>
     }]
    simSpaceSettings {maxAngularAcceleration=32, maxAngularVelocity=8, simSpaceAngularAlpha=1}
   }
   n14 = Anim Dynamics - Bone: dyn_hair_a_02(ComponentPose=@n13, Alpha=1, ExternalForce="0,0,0") {
    angularDampingOverride=0.9999
    angularSpringConstant=1000
    bAngularSpring=true
    bOverrideAngularDamping=true
    bUseGravityOverride=true
    boundBone {boneName=dyn_hair_a_02}
    componentAppliedLinearAccClamp=<5000 5000 5000>
    componentLinearAccScale=<0.5 0.5 0.5>
    componentLinearVelScale=<0.5 0.5 0.5>
    gravityOverride {z=-980}
    physicsBodyDefinitions [{
      boundBone {boneName=dyn_hair_a_02}
      boxExtents=<7 7 7>
      constraintSetup {angularLimitsMax=<5 30 30>, angularLimitsMin=<-5 -30 -30>, angularTarget=<1 0 0>}
      localJointOffset=<0.2 0 0>
     }]
    simSpaceSettings {maxAngularAcceleration=32, maxAngularVelocity=8, simSpaceAngularAlpha=1}
   }
   n15 = Anim Dynamics - Bone: dyn_hair_b_01(ComponentPose=@n14, Alpha=1, ExternalForce="0,0,0") {
    angularDampingOverride=0.9999
    angularSpringConstant=1000
    bAngularSpring=true
    bOverrideAngularDamping=true
    bUseGravityOverride=true
    boundBone {boneName=dyn_hair_b_01}
    componentAppliedLinearAccClamp=<5000 5000 5000>
    componentLinearAccScale=<0.5 0.5 0.5>
    componentLinearVelScale=<0.5 0.5 0.5>
    gravityOverride {z=-980}
    physicsBodyDefinitions [{
      boundBone {boneName=dyn_hair_b_01}
      boxExtents=<7 7 7>
      constraintSetup {angularLimitsMax=<5 30 30>, angularLimitsMin=<-5 -30 -30>, angularTarget=<1 0 0>}
      localJointOffset=<0.2 0 0>
     }]
    simSpaceSettings {maxAngularAcceleration=32, maxAngularVelocity=8, simSpaceAngularAlpha=1}
   }
   n16 = Anim Dynamics - Bone: dyn_hair_b_02(ComponentPose=@n15, Alpha=1, ExternalForce="0,0,0") {
    angularDampingOverride=0.9999
    angularSpringConstant=1000
    bAngularSpring=true
    bOverrideAngularDamping=true
    bUseGravityOverride=true
    boundBone {boneName=dyn_hair_b_02}
    componentAppliedLinearAccClamp=<5000 5000 5000>
    componentLinearAccScale=<0.5 0.5 0.5>
    componentLinearVelScale=<0.5 0.5 0.5>
    gravityOverride {z=-980}
    physicsBodyDefinitions [{
      boundBone {boneName=dyn_hair_b_02}
      boxExtents=<7 7 7>
      constraintSetup {angularLimitsMax=<5 30 30>, angularLimitsMin=<-5 -30 -30>, angularTarget=<1 0 0>}
      localJointOffset=<0.2 0 0>
     }]
    simSpaceSettings {maxAngularAcceleration=32, maxAngularVelocity=8, simSpaceAngularAlpha=1}
   }
   n17 = Anim Dynamics - Bone: dyn_hair_c_01(ComponentPose=@n16, Alpha=1, ExternalForce="0,0,0") {
    angularDampingOverride=0.9999
    angularSpringConstant=1000
    bAngularSpring=true
    bOverrideAngularDamping=true
    bUseGravityOverride=true
    boundBone {boneName=dyn_hair_c_01}
    componentAppliedLinearAccClamp=<5000 5000 5000>
    componentLinearAccScale=<0.5 0.5 0.5>
    componentLinearVelScale=<0.5 0.5 0.5>
    gravityOverride {z=-980}
    physicsBodyDefinitions [{
      boundBone {boneName=dyn_hair_c_01}
      boxExtents=<7 7 7>
      constraintSetup {angularLimitsMax=<5 30 30>, angularLimitsMin=<-5 -30 -30>, angularTarget=<1 0 0>}
      localJointOffset=<0.2 0 0>
     }]
    simSpaceSettings {maxAngularAcceleration=32, maxAngularVelocity=8, simSpaceAngularAlpha=1}
   }
   n18 = Anim Dynamics - Bone: dyn_hair_c_02(ComponentPose=@n17, Alpha=1, ExternalForce="0,0,0") {
    angularDampingOverride=0.9999
    angularSpringConstant=1000
    bAngularSpring=true
    bOverrideAngularDamping=true
    bUseGravityOverride=true
    boundBone {boneName=dyn_hair_c_02}
    componentAppliedLinearAccClamp=<5000 5000 5000>
    componentLinearAccScale=<0.5 0.5 0.5>
    componentLinearVelScale=<0.5 0.5 0.5>
    gravityOverride {z=-980}
    physicsBodyDefinitions [{
      boundBone {boneName=dyn_hair_c_02}
      boxExtents=<7 7 7>
      constraintSetup {angularLimitsMax=<5 30 30>, angularLimitsMin=<-5 -30 -30>, angularTarget=<1 0 0>}
      localJointOffset=<0.2 0 0>
     }]
    simSpaceSettings {maxAngularAcceleration=32, maxAngularVelocity=8, simSpaceAngularAlpha=1}
   }
   n19 = Anim Dynamics - Bone: dyn_hair_d_01(ComponentPose=@n18, Alpha=1, ExternalForce="0,0,0") {
    angularDampingOverride=0.9999
    angularSpringConstant=1000
    bAngularSpring=true
    bOverrideAngularDamping=true
    bUseGravityOverride=true
    boundBone {boneName=dyn_hair_d_01}
    componentAppliedLinearAccClamp=<5000 5000 5000>
    componentLinearAccScale=<0.5 0.5 0.5>
    componentLinearVelScale=<0.5 0.5 0.5>
    gravityOverride {z=-980}
    physicsBodyDefinitions [{
      boundBone {boneName=dyn_hair_d_01}
      boxExtents=<7 7 7>
      constraintSetup {angularLimitsMax=<5 30 30>, angularLimitsMin=<-5 -30 -30>, angularTarget=<1 0 0>}
      localJointOffset=<0.2 0 0>
     }]
    simSpaceSettings {maxAngularAcceleration=32, maxAngularVelocity=8, simSpaceAngularAlpha=1}
   }
   n20 = Anim Dynamics - Bone: dyn_hair_d_02(ComponentPose=@n19, Alpha=1, ExternalForce="0,0,0") {
    angularDampingOverride=0.9999
    angularSpringConstant=1000
    bAngularSpring=true
    bOverrideAngularDamping=true
    bUseGravityOverride=true
    boundBone {boneName=dyn_hair_d_02}
    componentAppliedLinearAccClamp=<5000 5000 5000>
    componentLinearAccScale=<0.5 0.5 0.5>
    componentLinearVelScale=<0.5 0.5 0.5>
    gravityOverride {z=-980}
    physicsBodyDefinitions [{
      boundBone {boneName=dyn_hair_d_02}
      boxExtents=<7 7 7>
      constraintSetup {angularLimitsMax=<5 30 30>, angularLimitsMin=<-5 -30 -30>, angularTarget=<1 0 0>}
      localJointOffset=<0.2 0 0>
     }]
    simSpaceSettings {maxAngularAcceleration=32, maxAngularVelocity=8, simSpaceAngularAlpha=1}
   }
   n21 = Anim Dynamics - Bone: dyn_hair_e_01(ComponentPose=@n20, Alpha=1, ExternalForce="0,0,0") {
    angularDampingOverride=0.9999
    angularSpringConstant=1000
    bAngularSpring=true
    bOverrideAngularDamping=true
    bUseGravityOverride=true
    boundBone {boneName=dyn_hair_e_01}
    componentAppliedLinearAccClamp=<5000 5000 5000>
    componentLinearAccScale=<0.5 0.5 0.5>
    componentLinearVelScale=<0.5 0.5 0.5>
    gravityOverride {z=-980}
    physicsBodyDefinitions [{
      boundBone {boneName=dyn_hair_e_01}
      boxExtents=<7 7 7>
      constraintSetup {angularLimitsMax=<5 30 30>, angularLimitsMin=<-5 -30 -30>, angularTarget=<1 0 0>}
      localJointOffset=<0.2 0 0>
     }]
    simSpaceSettings {maxAngularAcceleration=32, maxAngularVelocity=8, simSpaceAngularAlpha=1}
   }
   n22 = Anim Dynamics - Bone: dyn_hair_e_02(ComponentPose=@n21, Alpha=1, ExternalForce="0,0,0") {
    angularDampingOverride=0.9999
    angularSpringConstant=1000
    bAngularSpring=true
    bOverrideAngularDamping=true
    bUseGravityOverride=true
    boundBone {boneName=dyn_hair_e_02}
    componentAppliedLinearAccClamp=<5000 5000 5000>
    componentLinearAccScale=<0.5 0.5 0.5>
    componentLinearVelScale=<0.5 0.5 0.5>
    gravityOverride {z=-980}
    physicsBodyDefinitions [{
      boundBone {boneName=dyn_hair_e_02}
      boxExtents=<7 7 7>
      constraintSetup {angularLimitsMax=<5 30 30>, angularLimitsMin=<-5 -30 -30>, angularTarget=<1 0 0>}
      localJointOffset=<0.2 0 0>
     }]
    simSpaceSettings {maxAngularAcceleration=32, maxAngularVelocity=8, simSpaceAngularAlpha=1}
   }
   n23 = Anim Dynamics - Bone: dyn_hair_e_02_a(ComponentPose=@n22, Alpha=1, ExternalForce="0,0,0") {
    angularDampingOverride=0.9999
    angularSpringConstant=1000
    bAngularSpring=true
    bOverrideAngularDamping=true
    bUseGravityOverride=true
    boundBone {boneName=dyn_hair_e_02_a}
    componentAppliedLinearAccClamp=<5000 5000 5000>
    componentLinearAccScale=<0.5 0.5 0.5>
    componentLinearVelScale=<0.5 0.5 0.5>
    gravityOverride {z=-980}
    physicsBodyDefinitions [{
      boundBone {boneName=dyn_hair_e_02_a}
      boxExtents=<7 7 7>
      constraintSetup {angularLimitsMax=<5 30 30>, angularLimitsMin=<-5 -30 -30>, angularTarget=<1 0 0>}
      localJointOffset=<0.2 0 0>
     }]
    simSpaceSettings {maxAngularAcceleration=32, maxAngularVelocity=8, simSpaceAngularAlpha=1}
   }
   Output Pose(Result=Component To Local(ComponentPose=@n23)) {layerGroup=None}
  ]}
 EventGraph {nodes [Event BlueprintUpdateAnimation()]}
]
variables []
